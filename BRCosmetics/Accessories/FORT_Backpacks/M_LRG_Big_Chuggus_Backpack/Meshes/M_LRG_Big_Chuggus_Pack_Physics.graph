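# Physics asset for the Big Chuggus backpack cosmetic, giving its two tubes and a small strap bodies and joints for physical motion.
# Left and right tubes are mirrored seven-link chains of capsules with box tips, anchored to the pack capsule, with swing and twist limits tightening sharply away from the anchor.
# A two-box strap branches off the fifth left-tube link with looser limits.
# The pack joins pack_mid with free rotation, while a large root box is unconstrained.
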PhysicsAsset /BRCosmetics/Accessories/FORT_Backpacks/M_LRG_Big_Chuggus_Backpack/Meshes/M_LRG_Big_Chuggus_Pack_Physics
bodies {
 dyn_strap_01 [{shapeType=Box, extentX=4, extentY=1.3407, extentZ=4.1691, center=<1.9921 0 0>, rotation=<rot 90 0 0>}]
 dyn_strap_02 [{shapeType=Box, extentX=4, extentY=0.9571, extentZ=3.3988, center=<2.1212 0 0>, rotation=<rot 90 0 0>}]
 dyn_tube_l_01 [{shapeType=Capsule, length=4, radius=2, center=<3.0203 0 0>, rotation=<rot 90 0 0>}]
 dyn_tube_l_02 [{shapeType=Capsule, length=4, radius=2, center=<3.5472 0 0>, rotation=<rot 90 0 0>}]
 dyn_tube_l_03 [{shapeType=Capsule, length=4, radius=2, center=<3.1377 0 0>, rotation=<rot 90 0 0>}]
 dyn_tube_l_04 [{shapeType=Capsule, length=4, radius=2, center=<3.4339 0 0>, rotation=<rot 90 0 0>}]
 dyn_tube_l_05 [{shapeType=Capsule, length=4, radius=2, center=<3.7111 0 0>, rotation=<rot 90 0 0>}]
 dyn_tube_l_06 [{shapeType=Capsule, length=4, radius=2, center=<3.7401 0 0>, rotation=<rot 90 0 0>}]
 dyn_tube_l_07 [{shapeType=Box, extentX=1.01, extentY=1.01, extentZ=1.01, rotation=<rot 0 -90 0>}]
 dyn_tube_r_01 [{shapeType=Capsule, length=4, radius=2, center=<3.0532 0 0>, rotation=<rot 90 0 0>}]
 dyn_tube_r_02 [{shapeType=Capsule, length=4, radius=2, center=<3.8445 0 0>, rotation=<rot 90 0 0>}]
 dyn_tube_r_03 [{shapeType=Capsule, length=4, radius=2, center=<3.3117 0.0001 0>, rotation=<rot 90 0 0>}]
 dyn_tube_r_04 [{shapeType=Capsule, length=4, radius=2, center=<3.2397 0.0001 0>, rotation=<rot 90 0 0>}]
 dyn_tube_r_05 [{shapeType=Capsule, length=4, radius=2, center=<3.6792 0 0>, rotation=<rot 90 0 0>}]
 dyn_tube_r_06 [{shapeType=Capsule, length=4, radius=2, center=<2.5 0 0>, rotation=<rot 90 0 0>}]
 dyn_tube_r_07 [{shapeType=Box, extentX=1.01, extentY=1.01, extentZ=1.01, rotation=<rot 0 -90 0>}]
 pack [{shapeType=Capsule, length=18.074, radius=9.3687, center=<-0.0767 -9.7969 -3.5471>}]
 pack_mid [{shapeType=Capsule, length=0.505, radius=0.505}]
 root [{shapeType=Box, extentX=60.0438, extentY=7.5, extentZ=60.0438, center=<0.0001 4.2623 0>}]
}
constraints [
 {
  bone1Name=pack_mid
  bone2Name=pack
  swing1LimitDegrees=45
  swing1Motion=Free
  swing2LimitDegrees=45
  swing2Motion=Free
  twistLimitDegrees=45
  twistMotion=Free
 }
 {bone1Name=dyn_strap_02, bone2Name=dyn_strap_01, swing1LimitDegrees=45, swing2LimitDegrees=45, twistLimitDegrees=45}
 {bone1Name=dyn_strap_01, bone2Name=dyn_tube_l_05, swing1LimitDegrees=45, swing2LimitDegrees=45, twistLimitDegrees=45}
 {bone1Name=dyn_tube_l_07, bone2Name=dyn_tube_l_06, swing1LimitDegrees=1, swing2LimitDegrees=1, twistLimitDegrees=1}
 {bone1Name=dyn_tube_r_07, bone2Name=dyn_tube_r_06, swing1LimitDegrees=1, swing2LimitDegrees=1, twistLimitDegrees=1}
 {bone1Name=dyn_tube_r_01, bone2Name=pack, swing1LimitDegrees=20, swing2LimitDegrees=20, twistLimitDegrees=10}
 {bone1Name=dyn_tube_r_02, bone2Name=dyn_tube_r_01, swing1LimitDegrees=5, swing2LimitDegrees=5, twistLimitDegrees=3}
 {bone1Name=dyn_tube_r_03, bone2Name=dyn_tube_r_02, swing1LimitDegrees=2.2, swing2LimitDegrees=2.2, twistLimitDegrees=0.7}
 {bone1Name=dyn_tube_r_04, bone2Name=dyn_tube_r_03, swing1LimitDegrees=2.2, swing2LimitDegrees=2.2, twistLimitDegrees=0.7}
 {bone1Name=dyn_tube_r_05, bone2Name=dyn_tube_r_04, swing1LimitDegrees=2.2, swing2LimitDegrees=2.2, twistLimitDegrees=0.7}
 {bone1Name=dyn_tube_r_06, bone2Name=dyn_tube_r_05, swing1LimitDegrees=2.2, swing2LimitDegrees=2.2, twistLimitDegrees=0.7}
 {bone1Name=dyn_tube_l_01, bone2Name=pack, swing1LimitDegrees=20, swing2LimitDegrees=20, twistLimitDegrees=10}
 {bone1Name=dyn_tube_l_02, bone2Name=dyn_tube_l_01, swing1LimitDegrees=5, swing2LimitDegrees=5, twistLimitDegrees=3}
 {bone1Name=dyn_tube_l_03, bone2Name=dyn_tube_l_02, swing1LimitDegrees=2.2, swing2LimitDegrees=2.2, twistLimitDegrees=0.7}
 {bone1Name=dyn_tube_l_04, bone2Name=dyn_tube_l_03, swing1LimitDegrees=2.2, swing2LimitDegrees=2.2, twistLimitDegrees=0.7}
 {bone1Name=dyn_tube_l_05, bone2Name=dyn_tube_l_04, swing1LimitDegrees=2.2, swing2LimitDegrees=2.2, twistLimitDegrees=0.7}
 {bone1Name=dyn_tube_l_06, bone2Name=dyn_tube_l_05, swing1LimitDegrees=2.2, swing2LimitDegrees=2.2, twistLimitDegrees=0.7}
]
